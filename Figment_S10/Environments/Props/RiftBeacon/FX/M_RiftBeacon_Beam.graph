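# Translucent, unlit, two-sided material for the Rift Beacon beam effect, usable on mesh particles.
# SplineThicken sets beam thickness from base and tip width parameters.
# A tiled lightning plasma texture pans along V; its green channel blends edge and core colors, boosted by HDR intensity.
# Red plus faint blue channels form an opacity term, scaled by a parameter and a V-coordinate gradient along the beam.
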Material /Figment_S10/Environments/Props/RiftBeacon/FX/M_RiftBeacon_Beam
graph {
 properties {
  bEnableResponsiveAA=true
  bUsedWithMeshParticles=true
  blendMode=BLEND_Translucent
  pixelDepthOffsetMode=PDOM_Legacy
  shadingModel=MSM_Unlit
  twoSided=true
  usedShadingModels=Unlit
 }
 nodes [
  MaterialFunctionCall(
   WidthBase=ScalarParameter() {defaultValue=20, parameterName=Width Base}
   WidthTip=ScalarParameter() {defaultValue=100, parameterName=Width Tip}
  ) {materialFunction=&SplineThicken}
  n1 = TextureCoordinate()
  n2 = AppendVector(A=ScalarParameter() {defaultValue=1, parameterName=UTile}, B=ScalarParameter() {defaultValue=40, parameterName=VTile})
  n3 = Time()
  n4 = ScalarParameter() {defaultValue=1, parameterName=PanSpeed}
  n5 = TextureSample(UVs=Panner(Coordinate=Multiply(A=@n1, B=@n2), Time=Multiply(A=@n3, B=@n4)) {speedY=1}) {texture=/Game/Effects/Fort_Effects/Textures/Tile/T_Lightning_Plasma_Soft}
  n6 = Multiply(A=@n5.B) {constB=0.2}
  Multiply(
   A=ComponentMask(None=TextureCoordinate()) {g=true}
   B=Multiply(A=Add(A=@n5.R, B=@n6), B=ScalarParameter() {defaultValue=1, parameterName=Opacity})
  )
  Multiply(
   A=LinearInterpolate(
    A=VectorParameter() {parameterName=EdgeColor, defaultValue=<rgba 0.5 0.2226 0.2226 0>}.RGB
    B=VectorParameter() {parameterName=Color, defaultValue=<rgba 1 0.6214 0.6214 0>}.RGB
    Alpha=@n5.G
   )
   B=ScalarParameter() {defaultValue=20, parameterName=HDRIntensity}
  )
 ]
}
