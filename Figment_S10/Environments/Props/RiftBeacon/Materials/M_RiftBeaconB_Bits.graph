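# Lit surface material for the Rift Beacon B "bits" prop, combining base textures with animated emissive effects selected by vertex color.
# A timed, expanding sphere-mask ring blends with VisitorColor through the emissive texture, scaled by ArrayOn and suppressed by green vertex color.
# Red vertex areas blend to flickering Fresnel plus bump-offset, panning turbulence colored by curve-atlas ramps.
# The assembled attributes pass through the Geode hose-lights Rift Beacon function with the emissive texture as mask.
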Material /Figment_S10/Environments/Props/RiftBeacon/Materials/M_RiftBeaconB_Bits
graph {
 properties {bUseMaterialAttributes=true, pixelDepthOffsetMode=PDOM_Legacy, usedShadingModels=Default Lit}
 nodes [
  n1 = TextureSample() {samplerType=SAMPLERTYPE_Masks, texture=/Game/Environments/Props/RiftBeacon/Textures/T_RiftBeaconB_Bits_S}
  n2 = TextureCoordinate()
  n3 = ComponentMask(None=VectorParameter() {parameterName=LOcation, defaultValue=<rgba 0.4125 0.4 0 0>}.RGB) {g=true, r=true}
  n4 = OneMinus(None=Divide(A=Time() {bOverride_Period=true, period=5}, B=ScalarParameter() {defaultValue=2.5, parameterName=Pulse Speed}))
  n5 = Multiply(A=ScalarParameter() {defaultValue=0.7, parameterName=Radius}, B=@n4)
  n6 = SphereMask(A=@n2, B=@n3, Radius=@n5) {hardnessPercent=0}
  n7 = SphereMask(A=@n2, B=@n3, Radius=Subtract(A=@n5, B=ScalarParameter() {defaultValue=0.05, parameterName=Subtraction Amount}) {constB=0.01}, Hardness=@n4) {hardnessPercent=0}
  n8 = VectorParameter() {parameterName=ScreenColor, defaultValue=<rgba 0.0984 0.3942 1 1>}
  n9 = VectorParameter() {parameterName=VisitorColor, defaultValue=<rgba 1 0 0 0>}
  n10 = Multiply(
   A=Multiply(A=ScalarParameter() {defaultValue=1, parameterName=Emissive Intensity Increase}, B=Subtract(A=@n6, B=@n7))
   B=Multiply(A=LinearInterpolate(A=@n8.RGB, B=@n9.RGB, Alpha=@n4), B=@n4)
  )
  n11 = VectorParameter() {parameterName=VisitorColor, defaultValue=<rgba 1 0 0 0>}
  n12 = TextureSample() {texture=/Figment_S10/Environments/Props/RiftBeacon/Textures/T_RiftBeaconB_Bits_E}
  n13 = VertexColor()
  n14 = LinearInterpolate(A=Multiply(A=LinearInterpolate(A=@n10, B=@n11.RGB, Alpha=@n12.RGB), B=ScalarParameter() {parameterName=ArrayOn}), Alpha=@n13.G) {constB=0}
  n15 = ScalarParameter() {defaultValue=2, parameterName=Fresnel1}
  n16 = Multiply(A=Time(), B=ScalarParameter() {defaultValue=0.8, parameterName=FlickerSpeed})
  n17 = Sine(None=@n16) {period=0.55}
  n18 = Sine(None=@n16) {period=0.11}
  n19 = Add(A=Add(A=Multiply(B=@n17), B=Multiply(A=@n18) {constB=0.2}), B=Multiply(A=Sine(None=@n16) {period=0.175}) {constB=0.3})
  n20 = ConstantBiasScale(None=Multiply(A=Sine(None=@n16) {period=3.25}, B=Sine(None=@n16) {period=2.325}))
  n21 = Multiply(A=Multiply(A=Multiply(A=@n19, B=@n20)) {constB=0.3}) {constB=0.03}
  n22 = Multiply(
   A=VectorParameter() {parameterName=Outer, defaultValue=<rgba 0.2078 0.191 0.4844 1>}.RGB
   B=Fresnel(ExponentIn=Add(A=@n15, B=@n21)) {baseReflectFraction=0}
  )
  n23 = Multiply(A=Time(), B=ScalarParameter() {defaultValue=-1, parameterName=TexturePanningSpeed})
  n24 = ScalarParameter() {defaultValue=2, parameterName=U_Tile}
  n25 = ScalarParameter() {defaultValue=2, parameterName=V_Tile}
  n26 = Multiply(
   A=TextureCoordinate()
   B=AppendVector(A=ScalarParameter() {defaultValue=1, parameterName=U_Tile_2}, B=ScalarParameter() {defaultValue=1, parameterName=V_Tile_2})
  )
  n27 = TextureSample(UVs=Panner(Coordinate=@n26) {speedX=0.2, speedY=0.2}) {samplerType=SAMPLERTYPE_LinearColor, texture=/Game/Effects/Fort_Effects/Gameplay/SmokeDustExplosions/Textures/T_NoisePack_LinearColor}
  n28 = TextureSample(UVs=Panner(Coordinate=@n26) {speedX=-0.2, speedY=-0.2}) {samplerType=SAMPLERTYPE_LinearColor, texture=&T_NoisePack_LinearColor}
  n29 = ComponentMask(None=TextureCoordinate() {uTiling=0.5, vTiling=0.5}) {g=true}
  n30 = MaterialFunctionCall(In=Abs(None=Sine(None=@n29)), Contrast=Constant() {r=0.1}) {materialFunction=&CheapContrast}
  n31 = VertexNormalWS()
  n32 = CameraVectorWS()
  n33 = CurveAtlasRowParameter(CurveTime=Abs(None=DotProduct(A=@n31, B=@n32))) {
   atlas=&AthenaCurveAtlas01
   curve=/Game/Athena/Environments/Nexus/Materials/FXshader/Curve_Ramps
   defaultValue=79
  }
  n34 = Add(
   A=Multiply(A=@n30.Result, B=Subtract(A=@n27.R, B=@n28.G))
   B=Multiply(A=Subtract(A=@n33.R), B=MaterialFunctionCall() {materialFunction=&DebugTimeSine}.Result)
  )
  n35 = Multiply(A=@n30.Result, B=ScalarParameter() {defaultValue=-0.25, parameterName=Height})
  n36 = BumpOffset(Coordinate=Multiply(A=TextureCoordinate(), B=AppendVector(A=@n24, B=@n25)), Height=@n34, HeightRatioInput=@n35)
  n37 = Panner(Coordinate=@n36, Time=@n23) {speedX=0.1, speedY=-0.3}
  n38 = TextureSample(UVs=Panner(Coordinate=Multiply(A=@n36), Time=@n23) {speedX=0.1, speedY=-0.3}) {samplerType=SAMPLERTYPE_LinearColor, texture=&T_NoisePack_LinearColor}
  n39 = WorldPosition()
  n40 = Divide(
   A=ComponentMask(None=TransformPosition(None=@n39.XYZ) {transformSourceType=TRANSFORMPOSSOURCE_World}) {g=true, r=true}
   B=ScalarParameter() {defaultValue=1200, parameterName=Size}
  )
  n41 = LinearInterpolate(
   A=Multiply(A=TextureSample(UVs=@n37) {texture=/Game/Effects/Fort_Effects/Textures/Plasma/T_Turbulence_Seamless}.R, B=@n38.R)
   B=TextureSample(UVs=BumpOffset(Coordinate=@n40, Height=@n34, HeightRatioInput=@n35)) {texture=&T_Turbulence_Seamless}.R
   Alpha=OneMinus(None=@n30.Result)
  )
  n42 = Saturate(None=@n41)
  n43 = Multiply(A=@n41, B=ScalarParameter() {defaultValue=0.5, parameterName=NoiseScale})
  n44 = ScalarParameter() {defaultValue=1, parameterName=U_Tile_3}
  n45 = ScalarParameter() {defaultValue=1, parameterName=V_Tile_3}
  n46 = ComponentMask(None=Multiply(A=TextureCoordinate(), B=AppendVector(A=@n44, B=@n45))) {g=true}
  n47 = ScalarParameter() {defaultValue=-7, parameterName=RA_wide}
  n48 = Time()
  n49 = ScalarParameter() {defaultValue=2, parameterName=Panning Speed}
  n50 = Add(
   A=Add(A=Multiply(A=@n46, B=@n47), B=Multiply(A=@n48, B=@n49))
   B=Multiply(A=@n33.G, B=ScalarParameter() {defaultValue=0.025, parameterName=FresnelAffectTexture})
  )
  n51 = CurveAtlasRowParameter(CurveTime=Frac(None=Add(A=@n43, B=@n50))) {
   atlas=&AthenaCurveAtlas01
   curve=/Game/Athena/Environments/Nexus/Materials/FXshader/Curve_Color
   defaultValue=78
  }
  n52 = Add(
   A=Multiply(A=Multiply(A=@n42) {constB=5}, B=VectorParameter() {parameterName=TextureTint, defaultValue=<rgba 0.0984 0.3942 1 1>}.RGB)
   B=Power(Base=@n51, Exp=ScalarParameter() {defaultValue=1, parameterName=EdgeExp})
  )
  n53 = Multiply(
   A=Multiply(A=Add(A=@n22, B=@n52), B=ScalarParameter() {defaultValue=0.3, parameterName=RedVert_EmissScale})
   B=ScalarParameter() {defaultValue=1, parameterName=RedAlpha}
  )
  MaterialFunctionCall!(
   In=MakeMaterialAttributes!(
    BaseColor=TextureSample() {texture=/Game/Environments/Props/RiftBeacon/Textures/T_RiftBeaconB_Bits_D}.RGB
    Metallic
    Specular=@n1.R
    Roughness=@n1.R
    Anisotropy
    EmissiveColor=LinearInterpolate(A=@n14, B=@n53, Alpha=@n13.R)
    Opacity
    OpacityMask
    Normal=TextureSample() {samplerType=SAMPLERTYPE_Normal, texture=/Game/Environments/Props/RiftBeacon/Textures/T_RiftBeaconB_Bits_N}.RGB
    Tangent
    WorldPositionOffset
    SubsurfaceColor
    ClearCoat
    ClearCoatRoughness
    AmbientOcclusion
    Refraction
    CustomizedUVs_0
    CustomizedUVs_1
    CustomizedUVs_2
    CustomizedUVs_3
    CustomizedUVs_4
    CustomizedUVs_5
    CustomizedUVs_6
    CustomizedUVs_7
    PixelDepthOffset
    ShadingModel
    Displacement
   )
   Emissive Mask=TextureObject() {texture=&T_RiftBeaconB_Bits_E}
  ) {materialFunction=/Game/Effects/Fort_Effects/Materials/Geode/_Functions/MF_Geode_Hoses_Lights_RiftBeacon}
 ]
}
