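; Spline-based hose prop for the Rift Beacon that switches flow visuals and audio with named stage events.
; The construction script adds a spline mesh for each segment between spline points and stores it in an array.
; Stage00 disables the spline audio shape and zeroes HoseAlpha on every segment, while Stage02 enables both.
; When Invert Hose Flow Direction is set, Stage02 also overrides the Hose_Speed material parameter.
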(event EventBeginPlay
  (Shape|Disable (Variables|Default|GetFortSplineAudio)))

(event Custom|Stage00
  (Shape|Disable (Variables|Default|GetFortSplineAudio))
  (for _array_element (Variables|Default|GetSpline%20Meshes)
    (Rendering|Material|SetScalarParameterValueonMaterials _array_element "HoseAlpha" 0.0)))

(event Custom|Stage02
  (Shape|Enable (Variables|Default|GetFortSplineAudio))
  (for _array_element (Variables|Default|GetSpline%20Meshes)
    (Rendering|Material|SetScalarParameterValueonMaterials _array_element "HoseAlpha" 1.0)
    (if (Variables|Default|GetInvert%20Hose%20Flow%20Direction)
      (Rendering|Material|SetScalarParameterValueonMaterials _array_element "Hose_Speed" 2.0))))

(fn ConstructionScript
  (for _index (range (+ (- (Spline|GetNumberOfSplinePoints (Variables|Default|GetSpline)) 2) 1))
    (bind _returnvalue (AddComponent|NODE_AddSplineMeshComponent-0))
    (SplineMesh|SetStartandEnd _returnvalue (Spline|GetLocationandTangentatSplinePoint (Variables|Default|GetSpline) _index "Local") (.Tangent (Spline|GetLocationandTangentatSplinePoint (Variables|Default|GetSpline) _index "Local")) (Spline|GetLocationandTangentatSplinePoint (Variables|Default|GetSpline) (+ _index 1) "Local") (.Tangent (Spline|GetLocationandTangentatSplinePoint (Variables|Default|GetSpline) (+ _index 1) "Local")) true :!class /Script/Engine.SplineMeshComponent)
    (Utilities|Array|Add (Variables|Default|GetSpline%20Meshes) _returnvalue)))
