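; Rift Beacon prop for Season 10 that shows the calendar-event progress of a rift, from inactive through powered to beam stages.
; On match ready, StartActive or the placement and charging events (by hours elapsed) set Visible to 1–4.
; OnRep_Visible unhides the beacon, sets collision, and calls stage events on itself and the linked B and Hose actors.
; Stages drive material parameters, weld particles, shard visibility and staged ambient audio; Stage02 is unused by OnRep.
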(event Custom|CustomEvent)

(event Custom|DebugBeaconStage
  (params (Visible :type Integer))
  ;[1 1 1] Debugs Beacon stages using Start Ke*
  (Variables|Default|SetVisible Visible))

(event EventBeginPlay
  (Class|FortAsyncActionAuthorityMatchReady|MatchReadyAsync
    (:OnReady
      ;[1 1 1] Setting rift beacon stages based on how far the player is within the calender event
      (if (Variables|Default|GetStartActive?)
        (Variables|Default|SetVisible 4)
        (else
          (Utilities|FlowControl|Sequence
            (:then_0
              (bind _biseventactive (CalendarEvents|GetCalendarEventInformation (Variables|Default|GetCalenderPlacementEvent)))
              (if _biseventactive
                (Variables|Default|SetVisible 1)))
            (:then_1
              (bind (_biseventactive_2 _timesincebegin) (CalendarEvents|GetCalendarEventInformation (Variables|Default|GetCalenderChargingEvent)))
              (if _biseventactive_2
                (if (>= (Math|Timespan|GetTotalHours _timesincebegin) 24.0)
                  (Variables|Default|SetVisible 4)
                  (elif (>= (Math|Timespan|GetTotalHours _timesincebegin) 12.0)
                    (Variables|Default|SetVisible 3)
                    (elif (>= (Math|Timespan|GetTotalHours _timesincebegin) 0.0)
                      (Variables|Default|SetVisible 2))))))))))))

(event Custom|Stage%2000
  (bind _s_fx_riftbeacon_internalfx (Variables|Default|GetS_FX_RiftBeacon_InternalFX))
  (bind _s_riftbeacon_a (Variables|Default|GetS_RiftBeacon_A))
  ;[1 1 1] Placed in World - NON ACTIVE
  (Components|Activation|Deactivate (Variables|Default|GetP_RiftBeacon_WeldPoint))
  (Rendering|Material|SetScalarParameterValueonMaterials _s_fx_riftbeacon_internalfx "Opacity" 0.0)
  (Rendering|Material|SetScalarParameterValueonMaterials _s_fx_riftbeacon_internalfx "BlueAlpha" 0.0)
  (Rendering|Material|SetScalarParameterValueonMaterials _s_riftbeacon_a "RedAlpha" 0.0)
  (Rendering|Material|SetScalarParameterValueonMaterials _s_riftbeacon_a "HoseAlpha" 0.0)
  (Rendering|Material|SetScalarParameterValueonMaterials _s_riftbeacon_a "VibrationOn" 0.0)
  (Rendering|Material|SetScalarParameterValueonMaterials _s_riftbeacon_a "Freeze" 0.0)
  (Rendering|Material|SetVectorParameterValueonMaterials _s_riftbeacon_a "Emissive3" "0, 0, 0")
  (Rendering|Material|SetScalarParameterValueonMaterials _s_riftbeacon_a "MonitorOn" 0.0)
  (Rendering|SetVisibility (Variables|Default|GetSM_ZeroPoint_Shard_L) false false))

(event Custom|Stage%2001
  (bind _s_fx_riftbeacon_internalfx (Variables|Default|GetS_FX_RiftBeacon_InternalFX))
  (bind _s_riftbeacon_a (Variables|Default|GetS_RiftBeacon_A))
  ;[1 1 1] Power Turned On, but no Shard Element or Beam
  (Components|Activation|Deactivate (Variables|Default|GetP_RiftBeacon_WeldPoint))
  (Rendering|Material|SetScalarParameterValueonMaterials _s_fx_riftbeacon_internalfx "Opacity" 1.0)
  (Rendering|Material|SetScalarParameterValueonMaterials _s_fx_riftbeacon_internalfx "BlueAlpha" 0.0)
  (Rendering|Material|SetScalarParameterValueonMaterials _s_riftbeacon_a "RedAlpha" 1.0)
  (Rendering|Material|SetScalarParameterValueonMaterials _s_riftbeacon_a "HoseAlpha" 1.0)
  (Rendering|Material|SetScalarParameterValueonMaterials _s_riftbeacon_a "VibrationOn" 1.0)
  (Rendering|Material|SetScalarParameterValueonMaterials _s_riftbeacon_a "Freeze" 1.0)
  (Rendering|Material|SetVectorParameterValueonMaterials _s_riftbeacon_a "Emissive3" "40.000000, 0, 0")
  (Rendering|Material|SetScalarParameterValueonMaterials _s_riftbeacon_a "MonitorOn" 1.0)
  (Rendering|SetVisibility (Variables|Default|GetSM_ZeroPoint_Shard_L) false false)
  (CallFunction|StartAudio 2))

(event Custom|Stage%2002
  (bind _s_fx_riftbeacon_internalfx (Variables|Default|GetS_FX_RiftBeacon_InternalFX))
  (bind _s_riftbeacon_a (Variables|Default|GetS_RiftBeacon_A))
  ;[1 1 1] Power Turned On, Shard Element Added No Beam
  (Components|Activation|Activate (Variables|Default|GetP_RiftBeacon_WeldPoint) false)
  (Rendering|Material|SetScalarParameterValueonMaterials _s_fx_riftbeacon_internalfx "Opacity" 1.0)
  (Rendering|Material|SetScalarParameterValueonMaterials _s_fx_riftbeacon_internalfx "BlueAlpha" 1.0)
  (Rendering|Material|SetScalarParameterValueonMaterials _s_riftbeacon_a "RedAlpha" 1.0)
  (Rendering|Material|SetScalarParameterValueonMaterials _s_riftbeacon_a "HoseAlpha" 1.0)
  (Rendering|Material|SetScalarParameterValueonMaterials _s_riftbeacon_a "VibrationOn" 1.0)
  (Rendering|Material|SetScalarParameterValueonMaterials _s_riftbeacon_a "Freeze" 1.0)
  (Rendering|Material|SetVectorParameterValueonMaterials _s_riftbeacon_a "Emissive3" "40.000000, 0, 0")
  (Rendering|Material|SetScalarParameterValueonMaterials _s_riftbeacon_a "MonitorOn" 1.0)
  (Rendering|SetVisibility (Variables|Default|GetSM_ZeroPoint_Shard_L) true false))

(event Custom|Stage%2003
  (bind _s_fx_riftbeacon_internalfx (Variables|Default|GetS_FX_RiftBeacon_InternalFX))
  (bind _s_riftbeacon_a (Variables|Default|GetS_RiftBeacon_A))
  ;[1 1 1] Beam Active, Rift Formed
  (Components|Activation|Activate (Variables|Default|GetP_RiftBeacon_WeldPoint) false)
  (Rendering|Material|SetScalarParameterValueonMaterials _s_fx_riftbeacon_internalfx "Opacity" 1.0)
  (Rendering|Material|SetScalarParameterValueonMaterials _s_fx_riftbeacon_internalfx "BlueAlpha" 1.0)
  (Rendering|Material|SetScalarParameterValueonMaterials _s_riftbeacon_a "RedAlpha" 1.0)
  (Rendering|Material|SetScalarParameterValueonMaterials _s_riftbeacon_a "HoseAlpha" 1.0)
  (Rendering|Material|SetScalarParameterValueonMaterials _s_riftbeacon_a "VibrationOn" 1.0)
  (Rendering|Material|SetScalarParameterValueonMaterials _s_riftbeacon_a "Freeze" 1.0)
  (Rendering|Material|SetVectorParameterValueonMaterials _s_riftbeacon_a "Emissive3" "40.000000, 0, 0")
  (Rendering|Material|SetScalarParameterValueonMaterials _s_riftbeacon_a "MonitorOn" 1.0)
  (Rendering|SetVisibility (Variables|Default|GetSM_ZeroPoint_Shard_L) true false)
  (CallFunction|StartAudio 3))

(event Custom|Stage%2004
  (bind _s_riftbeacon_a (Variables|Default|GetS_RiftBeacon_A))
  (bind _s_fx_riftbeacon_internalfx (Variables|Default|GetS_FX_RiftBeacon_InternalFX))
  ;[1 1 1] Beam Active, Rift Full
  (Components|Activation|Activate (Variables|Default|GetP_RiftBeacon_WeldPoint) false)
  (Rendering|Material|SetScalarParameterValueonMaterials _s_fx_riftbeacon_internalfx "Opacity" 1.0)
  (Rendering|Material|SetScalarParameterValueonMaterials _s_fx_riftbeacon_internalfx "BlueAlpha" 1.0)
  (Rendering|Material|SetScalarParameterValueonMaterials _s_riftbeacon_a "RedAlpha" 1.0)
  (Rendering|Material|SetScalarParameterValueonMaterials _s_riftbeacon_a "HoseAlpha" 1.0)
  (Rendering|Material|SetScalarParameterValueonMaterials _s_riftbeacon_a "VibrationOn" 1.0)
  (Rendering|Material|SetScalarParameterValueonMaterials _s_riftbeacon_a "Freeze" 1.0)
  (Rendering|Material|SetVectorParameterValueonMaterials _s_riftbeacon_a "Emissive3" "40.000000, 0, 0")
  (Rendering|Material|SetScalarParameterValueonMaterials _s_riftbeacon_a "MonitorOn" 1.0)
  (Rendering|SetVisibility (Variables|Default|GetSM_ZeroPoint_Shard_L) true false)
  (CallFunction|StartAudio 3))

(fn OnRep_Visible
  (switch int (Variables|Default|GetVisible)
    (:1
      (CallFunction|SetVisibilityandCollision (Variables|Default|GetRiftBeaconA) (Variables|Default|GetRiftBeaconB) (Variables|Default|GetRiftBeaconHose) (Variables|Default|GetS_RiftBeacon_A (Variables|Default|GetRiftBeaconA)) (Variables|/Figment_S10/Environments/Props/RiftBeacon/RiftBeacon_B.RiftBeacon_B_C|GetS_RiftBeacon_B (Variables|Default|GetRiftBeaconB)) false true)
      (CallFunction|Stage00 self)
      (Class|RiftBeaconB|Stage00 (Variables|Default|GetRiftBeaconB))
      (Class|RiftBeaconHose|Stage00 (Variables|Default|GetRiftBeaconHose)))
    (:2
      (CallFunction|SetVisibilityandCollision (Variables|Default|GetRiftBeaconA) (Variables|Default|GetRiftBeaconB) (Variables|Default|GetRiftBeaconHose) (Variables|Default|GetS_RiftBeacon_A (Variables|Default|GetRiftBeaconA)) (Variables|/Figment_S10/Environments/Props/RiftBeacon/RiftBeacon_B.RiftBeacon_B_C|GetS_RiftBeacon_B (Variables|Default|GetRiftBeaconB)) false true)
      (CallFunction|Stage01 self)
      (Class|RiftBeaconB|Stage02 (Variables|Default|GetRiftBeaconB))
      (Class|RiftBeaconHose|Stage02 (Variables|Default|GetRiftBeaconHose)))
    (:3
      (CallFunction|SetVisibilityandCollision (Variables|Default|GetRiftBeaconA) (Variables|Default|GetRiftBeaconB) (Variables|Default|GetRiftBeaconHose) (Variables|Default|GetS_RiftBeacon_A (Variables|Default|GetRiftBeaconA)) (Variables|/Figment_S10/Environments/Props/RiftBeacon/RiftBeacon_B.RiftBeacon_B_C|GetS_RiftBeacon_B (Variables|Default|GetRiftBeaconB)) false true)
      (CallFunction|Stage03 self)
      (Class|RiftBeaconB|Stage02 (Variables|Default|GetRiftBeaconB))
      (Class|RiftBeaconHose|Stage02 (Variables|Default|GetRiftBeaconHose)))
    (:4
      (CallFunction|SetVisibilityandCollision (Variables|Default|GetRiftBeaconA) (Variables|Default|GetRiftBeaconB) (Variables|Default|GetRiftBeaconHose) (Variables|Default|GetS_RiftBeacon_A (Variables|Default|GetRiftBeaconA)) (Variables|/Figment_S10/Environments/Props/RiftBeacon/RiftBeacon_B.RiftBeacon_B_C|GetS_RiftBeacon_B (Variables|Default|GetRiftBeaconB)) false true)
      (CallFunction|Stage04 self)
      (Class|RiftBeaconB|Stage02 (Variables|Default|GetRiftBeaconB))
      (Class|RiftBeaconHose|Stage02 (Variables|Default|GetRiftBeaconHose)))))

(fn SetVisibilityAndCollision
  (params (RiftBeaconA :type Actor%20Object%20Reference) (RiftBeaconB :type Actor%20Object%20Reference) (RiftBeaconHose :type Actor%20Object%20Reference) (RiftBeaconA_SM :type Primitive%20Component%20Object%20Reference) (RiftBeaconB_SM :type Primitive%20Component%20Object%20Reference) (bNewHidden :type Boolean) (bNewActorEnableCollision :type Boolean))
  (Rendering|SetActorHiddenInGame RiftBeaconA bNewHidden)
  (Collision|SetActorEnableCollision RiftBeaconA bNewActorEnableCollision)
  (Collision|SetCollisionEnabled RiftBeaconB_SM "QueryAndPhysics")
  (Collision|SetCollisionResponsetoAllChannels RiftBeaconB_SM "ECR_Block"))

(fn StartAudio
  (params (Stage :type Integer))
  (bind _ambientaudio (Variables|Default|GetAmbientAudio))
  (Audio|Components|Audio|FadeIn _ambientaudio 1.0 1.0 0.0 "Linear" :!class /Script/Engine.AudioComponent)
  (Audio|Parameter|SetIntegerParameter _ambientaudio "Stage" Stage :!class /Script/Engine.AudioComponent))

(fn ConstructionScript)
