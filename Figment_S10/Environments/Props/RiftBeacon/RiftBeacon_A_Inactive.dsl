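; Rift Beacon A prop in its inactive variant, an environment actor that sets up the beacon in its base stage.
; On BeginPlay it unhides the actor, makes the S_RiftBeacon_A component block all channels with query and physics collision, then calls Stage00.
; The StagesBeaconA rep-notify handles only stage 0, calling a separate SetVisibilityandCollision function and Stage00, neither defined here.
; The construction script just calls its parent.
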(event EventBeginPlay
  ;[1 1 1] Start Disabled
  (CallFunction|SetVisibilityCollision self)
  (CallFunction|Stage00 self))

(fn OnRep_StagesBeaconA
  (switch int (Variables|Default|GetStagesBeaconA)
    (:0
      (CallFunction|SetVisibilityandCollision self 0 0 0 0 false true)
      (CallFunction|Stage00 self))))

(fn SetVisibilityCollision
  (Rendering|SetActorHiddenInGame false)
  (Collision|SetActorEnableCollision true)
  (Collision|SetCollisionEnabled (Variables|Default|GetS_RiftBeacon_A) "QueryAndPhysics")
  (Collision|SetCollisionResponsetoAllChannels (Variables|Default|GetS_RiftBeacon_A) "ECR_Block"))

(fn ConstructionScript
  (|CallParentFunction))
