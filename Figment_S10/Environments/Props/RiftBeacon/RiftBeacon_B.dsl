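; RiftBeacon_B is a rift beacon prop with two custom stage events, 0 and 2, that switch its array off and on.
; Each stage toggles the arc particle component, sets array, red, hose and vibration material scalars on the beacon mesh, and calls StartAudio.
; StartAudio fades in the ambient audio and passes the stage as an integer parameter.
; SpawnDecals events from the Collect particles write their world-space location into the material's Particle Hit vector.
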(event Custom|Stage%2000
  (bind _s_riftbeacon_b (Variables|Default|GetS_RiftBeacon_B))
  ;[1 1 1] Array OFF
  (Components|Activation|Deactivate (Variables|Default|GetP_RiftBeaconB_ArcsB))
  (Rendering|Material|SetScalarParameterValueonMaterials _s_riftbeacon_b "ArrayOn" 0.0)
  (Rendering|Material|SetScalarParameterValueonMaterials _s_riftbeacon_b "RedAlpha" 0.0)
  (Rendering|Material|SetScalarParameterValueonMaterials _s_riftbeacon_b "HoseAlpha" 0.0)
  (Rendering|Material|SetScalarParameterValueonMaterials _s_riftbeacon_b "VibrationOn" 0.0)
  (CallFunction|StartAudio 0))

(event Custom|Stage%2002
  (bind _s_riftbeacon_b (Variables|Default|GetS_RiftBeacon_B))
  ;[1 1 1] Array ON
  (Components|Activation|Activate (Variables|Default|GetP_RiftBeaconB_ArcsB) false)
  (Rendering|Material|SetScalarParameterValueonMaterials _s_riftbeacon_b "ArrayOn" 1.0)
  (Rendering|Material|SetScalarParameterValueonMaterials _s_riftbeacon_b "RedAlpha" 1.0)
  (Rendering|Material|SetScalarParameterValueonMaterials _s_riftbeacon_b "HoseAlpha" 1.0)
  (Rendering|Material|SetScalarParameterValueonMaterials _s_riftbeacon_b "VibrationOn" 1.0)
  (CallFunction|StartAudio 2))

(event OnParticleSpawn(P_RiftBeaconB_Collect)
  (params (EventName :type Name :default "None") (EmitterTime :type Float%20%28single-precision%29) (Location :type Vector :default "0, 0, 0") (Velocity :type Vector :default "0, 0, 0"))
  (switch name EventName
    (:SpawnDecals
      (Rendering|Material|SetVectorParameterValueonMaterials (Variables|Default|GetS_RiftBeacon_B) "Particle Hit" (Math|Transform|TransformLocation (Transformation|GetWorldTransform (Variables|Default|GetP_RiftBeaconB_Collect)) Location)))))

(fn StartAudio
  (params (Stage :type Integer))
  (bind _ambientaudio (Variables|Default|GetAmbientAudio))
  (Audio|Components|Audio|FadeIn _ambientaudio 1.0 1.0 0.0 "Linear" :!class /Script/Engine.AudioComponent)
  (Audio|Parameter|SetIntegerParameter _ambientaudio "Stage" Stage :!class /Script/Engine.AudioComponent))

(fn ConstructionScript)
