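; Blueprint for the inactive Rift Beacon B prop, which starts in its initial stage with visibility and blocking collision set on.
; On BeginPlay it calls SetVisibilityAndCollision, then Stage00, whose body is not defined here.
; The StagesBeaconB replication callback repeats both calls when the stage is 0, the only case it handles.
; SetVisibilityAndCollision unhides the actor and makes the S_RiftBeacon_B component block all channels with query and physics collision.
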(event EventBeginPlay
  ;[1 1 1] Start Disabled
  (CallFunction|SetVisibilityandCollision self)
  (CallFunction|Stage00 self))

(fn OnRep_StagesBeaconB
  (switch int (Variables|Default|GetStagesBeaconB)
    (:0
      (CallFunction|SetVisibilityandCollision)
      (CallFunction|Stage00))))

(fn SetVisibilityAndCollision
  (Rendering|SetActorHiddenInGame false)
  (Collision|SetActorEnableCollision true)
  (Collision|SetCollisionEnabled (Variables|Default|GetS_RiftBeacon_B) "QueryAndPhysics")
  (Collision|SetCollisionResponsetoAllChannels (Variables|Default|GetS_RiftBeacon_B) "ECR_Block"))

(fn ConstructionScript
  (|CallParentFunction))
