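; Rift beacon actor that stages a beam, sky rift and active rift bubble visuals and audio during a calendar event.
; On authority match ready, elapsed lead-up event hours or the map-active event set the Visible integer to 3, 4 or 5.
; OnRep_Visible unhides the actor and runs that stage, switching particles, rift opacity, aurora and audio parameter.
; The construction script fits beam particles, rift position and audio attenuation to a beam height variable.
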(event Custom|DebugBeaconStage
  (params (Visible :type Integer))
  ;[1 1 1] Debugs Beacon stages using Ke*
  (Variables|Default|SetVisible Visible))

(event EventBeginPlay
  ;[1 1 1] Setting rift beacon stages based on how far the player is within the calender event
  (Class|FortAsyncActionAuthorityMatchReady|MatchReadyAsync
    (:OnReady
      (Utilities|FlowControl|Sequence
        (:then_0
          (bind (_eventname _timeuntilend _timesincebegin) (MCPContext|WatchCalendarEvent (Variables|Default|GetCalenderLeadUpEvent))
            (:OnEventActive
              (if (>= (Math|Timespan|GetTotalHours _timesincebegin) 36.0)
                (Variables|Default|SetVisible 4)
                (elif (>= (Math|Timespan|GetTotalHours _timesincebegin) 24.0)
                  (Variables|Default|SetVisible 3))))))
        (:then_1
          (MCPContext|WatchCalendarEvent (Variables|Default|GetCalenderMapActiveEvent)
            (:OnEventActive
              (Variables|Default|SetVisible 5))))))))

(event Custom|Stage%2003
  (bind _rift (Variables|Default|GetRIFT))
  ;[1 1 1] Beam Active, Rift Formed
  (Components|Activation|Activate (Variables|Default|GetP_Riftbeacon_Beam) false)
  (Rendering|SetVisibility _rift true false)
  (Rendering|Material|SetScalarParameterValueonMaterials _rift "Opacity" 0.1)
  (CallFunction|StartAudio 3)
  (Rendering|SetVisibility (Variables|Default|GetAurora) false false)
  (Components|Activation|Deactivate (Variables|Default|GetP_Geode_Crack_Center_ParticlesOnly)))

(event Custom|Stage%2004
  (bind _rift (Variables|Default|GetRIFT))
  ;[1 1 1] Beam Active, Rift Full
  (Components|Activation|Activate (Variables|Default|GetP_Riftbeacon_Beam) false)
  (Rendering|SetVisibility _rift true false)
  (Rendering|Material|SetScalarParameterValueonMaterials _rift "Opacity" 1.0)
  (CallFunction|StartAudio 4)
  (Rendering|SetVisibility (Variables|Default|GetAurora) false false)
  (Components|Activation|Deactivate (Variables|Default|GetP_Geode_Crack_Center_ParticlesOnly)))

(event Custom|Rift%20Bubble%20Active
  (bind _rift (Variables|Default|GetRIFT))
  ;[1 1 1] When Rift Bubble and new POI Gameplay is Active
  (Components|Activation|Deactivate (Variables|Default|GetP_Riftbeacon_Beam))
  (Components|Activation|Activate (Variables|Default|GetP_Geode_Crack_Center_ParticlesOnly) false)
  (Rendering|SetVisibility _rift true false)
  (Rendering|Material|SetScalarParameterValueonMaterials _rift "Opacity" 1.0)
  (Rendering|SetVisibility (Variables|Default|GetAurora) true false)
  (CallFunction|StartAudio 5))

(fn OnRep_Visible
  (bind _self self)
  ;[1 1 1] Set stages based on int value
  (switch int (Variables|Default|GetVisible)
    (:3
      (Rendering|SetActorHiddenInGame _self false)
      (CallFunction|Stage03))
    (:4
      (Rendering|SetActorHiddenInGame _self false)
      (CallFunction|Stage04))
    (:5
      (Rendering|SetActorHiddenInGame _self false)
      (CallFunction|RiftBubbleActive))))

(fn StartAudio
  (params (Stage :type Integer))
  (bind _ambientaudiobeam (Variables|Default|GetAmbientAudioBeam))
  (Audio|Components|Audio|FadeIn _ambientaudiobeam 1.0 1.0 0.0 "Linear" :!class /Script/Engine.AudioComponent)
  (Audio|Parameter|SetIntegerParameter _ambientaudiobeam "Stage" Stage :!class /Script/Engine.AudioComponent))

(fn ConstructionScript
  (bind (_benablepriorityattenuation _benablesubmixsends _benablesourcedataoverride _benablesendtoaudiolink _audiolinksettingsoverride _binauralradius _priorityattenuationmethod _nonspatializedradiusstart _nonspatializedradiusend _nonspatializedradiusmode _priorityattenuationmin _priorityattenuationmax _priorityattenuationdistancemin _priorityattenuationdistancemax _manualpriorityattenuation _submixsendsettings _custompriorityattenuationcurve _falloffmode _attenuationshapeextents) (Utilities|Struct|BreakSoundAttenuationSettings (Variables|/Script/Engine.AudioComponent|GetAttenuationOverrides (Variables|Default|GetAmbientAudioBeam))))
  (bind (_attenuationshapeextents_x _attenuationshapeextents_y _attenuationshapeextents_z) (Math|Vector|BreakVector _attenuationshapeextents))
  (bind _rift_beacon_beam_height (Variables|Default|GetRift%20Beacon%20Beam%20Height))
  (bind _ambientaudiobeam (Variables|Default|GetAmbientAudioBeam))
  (bind _p_riftbeacon_beam (Variables|Default|GetP_Riftbeacon_Beam))
  (Effects|Components|ParticleSystem|SetVectorParameter _p_riftbeacon_beam "SizeScale" (Math|Vector|MakeVector (/ _rift_beacon_beam_height 1e+03) 1.0 1.0))
  (Effects|Components|ParticleSystem|SetVectorParameter _p_riftbeacon_beam "GlintLocation" (Math|Vector|MakeVector _rift_beacon_beam_height))
  (Transformation|SetRelativeLocation (Variables|Default|GetRIFT) (Math|Vector|MakeVector 0.0 0.0 (Variables|Default|GetRift%20Beacon%20Beam%20Height)) false false)
  (bind _structout (Utilities|Struct|SetmembersinSoundAttenuationSettings (Variables|/Script/Engine.AudioComponent|GetAttenuationOverrides _ambientaudiobeam) (Math|Vector|MakeVector (/ (Variables|Default|GetRift%20Beacon%20Beam%20Height) 2.0) _attenuationshapeextents_y _attenuationshapeextents_z)))
  (Audio|Components|Audio|AdjustAttenuation _ambientaudiobeam _structout)
  (Transformation|SetRelativeLocation (Variables|Default|GetAmbientAudioBeam) (Math|Vector|MakeVector 0.0 0.0 (/ (Variables|Default|GetRift%20Beacon%20Beam%20Height) 2.0)) false false)
  (Transformation|SetRelativeLocation (Variables|Default|GetAmbientAudioRift) (Math|Vector|MakeVector 0.0 0.0 (Variables|Default|GetRift%20Beacon%20Beam%20Height)) false false))
